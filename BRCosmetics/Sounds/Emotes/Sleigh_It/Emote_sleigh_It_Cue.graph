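# Sound cue for the looping first-person music of the Sleigh It emote.
# A Mixer blends two looping wave players, playing the sleigh bells layer at half the level of the music track.
# It uses the first-person emote music sound class and concurrency rules, and sends to a music audio-analysis submix.
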SoundCue /BRCosmetics/Sounds/Emotes/Sleigh_It/Emote_sleigh_It_Cue
root=Mixer_0
nodes [
 {id=Mixer_0, children [WavePlayer_2,WavePlayer_3], properties {inputVolume [0.5,1]}}
 {id=WavePlayer_2, properties {bLooping=true, soundWaveAssetPtr=/Game/Athena/Sounds/Emotes/Sleigh_It/Emote_Sleigh_It_Bells_02}}
 {id=WavePlayer_3, properties {bLooping=true, soundWaveAssetPtr=&/Sleigh_It_Emote_Music_01}}
]
properties {
 maxDistance=2097152
 soundClassObject=/Game/Sounds/FortSoundClassesAndModes/Fort_Music_Emote_1P_PSM
 volumeMultiplier=1.1
 concurrencyOverrides {maxCount=3, resolutionRule=StopOldest}
 concurrencySet [/Game/Sounds/Concurrency_Rules/Athena_EmoteMusic_1P_Conc]
 soundSubmixSends [{minSendLevel=0, soundSubmix=&Fortnite_AudioAnalysis_Music}]
}
