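# Third-person sound cue for the Sleigh It emote, presumably what plays from other players performing it.
# A mixer blends two looping waves, sleigh bells at half volume and the emote music at full volume.
# It uses emote attenuation, the third-person emote music sound class and emote-music concurrency, with an override that prevents new instances.
# It keeps playing when silent and sends to a music audio-analysis submix.
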SoundCue /BRCosmetics/Sounds/Emotes/Sleigh_It/SC_Emote_Sleigh_It_3P_Cue
root=Mixer_0
nodes [
 {id=Mixer_0, children [WavePlayer_1,WavePlayer_2], properties {inputVolume [0.5,1]}}
 {id=WavePlayer_1, properties {bLooping=true, soundWaveAssetPtr=/Game/Athena/Sounds/Emotes/Sleigh_It/Emote_Sleigh_It_Bells_02}}
 {id=WavePlayer_2, properties {bLooping=true, soundWaveAssetPtr=&/Sleigh_It_Emote_Music_01}}
]
properties {
 attenuationSettings=&Fort_Emotes_Athena_Att
 maxDistance=1600
 soundClassObject=/Game/Sounds/FortSoundClassesAndModes/Fort_Music_Emote_3P
 virtualizationMode=PlayWhenSilent
 volumeMultiplier=1.1
 concurrencyOverrides {resolutionRule=PreventNew}
 concurrencySet [/Game/Sounds/Concurrency_Rules/Athena_EmoteMusic_Conc]
 soundSubmixSends [{minSendLevel=0, soundSubmix=&Fortnite_AudioAnalysis_Music}]
}
